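# Niagara deploy effect for the CoolSlice Golf glider, with separate timing and tuning for front end and gameplay.
# Hairs bursts sprites from a socket after a delay, tinted by time of day and moved by velocity, gravity, drag and curl noise.
# MotionBlur bursts a mesh that follows another socket, fading through alpha curves scaled by a system time-of-day value.
# Both emitters sample a shared system Socket Reader.
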NiagaraSystem /BRCosmetics/Gadgets/Assets/VinderTech_GliderChute/Glider_CoolSlice_Golf/FX/NS_Glider_CoolSlice_Deploy
dependencies {
 usedDataInterfaces [&NiagaraDataInterfaceCurve, &NiagaraDataInterfaceSocketReader]
 usedDynamicInputs [&SetBoolByFloatComparison, &FNTimeOfDayScalar]
 usedModules [
  &self:SystemSpawnScript.SystemScriptSource.SystemScriptGraph.NiagaraNodeAssignment_0.SetVariables_#1
  &SystemState
  &FN_Timer
  &self:SystemSpawnScript.SystemScriptSource.SystemScriptGraph.NiagaraNodeAssignment_1.SetVariables_#2
 ]
}
emitters {
 Hairs {
  emitterUpdateScript [
   &EmitterState(
    Life Cycle Mode=::NewEnumerator1
    Loop Behavior=::NewEnumerator1
    Loop Delay=&FN_FrontEndFloatModifier(Gameplay=1.25, Front End=1.3)
    Loop Duration=1
    UseLoopDelay=true
   )
   &SpawnBurst_Instantaneous(
    Spawn Count=&RandomRangeInt(Minimum=35, Maximum=50)
    Use Spawn Probability=false
    UseLoopCountLimit=false
   )
  ]
  particleSpawnScript [
   &InitializeParticle(
    Color=&FNTimeOfDayColor(
     Night Color=<rgba 0.3 0.3608 1 1>
     Morning Color=<rgba 1.25 1.25 0.375 1>
     Day Color=&MakeCustomLinearColorFromBool(
      Bool=&Boolean_OrOperation(Bool1=$User.bIsFrontEnd, Bool2=$User.bIsFrontEndPreview)
      True Color=<rgba 0.6145 0.6145 0.6145 1>
      False Color=<rgba 1 1 1 1>
     )
     Evening Color=<rgba 1 0.8017 0.4051 1>
    )
    Lifetime Mode=::NewEnumerator1
    Sprite Rotation Mode=::NewEnumerator2
    Sprite Size=&Multiply_Vector2D_ByFloat(Float=&RandomRangeFloat(Minimum=9, Maximum=12), Vector2D=<1 2>)
    Sprite Size Mode=::NewEnumerator4
    UsePositionOffset=false
    Write Material Random=true
   )
   &SubUVAnimation(End Frame=1, SubUV Animation Mode:ENiagara_SubUVLookupMode=::NewEnumerator2)
   &SocketLocation(
    Inherited Velocity Speed Limit=25
    SamplePosOffset=true
    Sampled Position Offset=<0 75 0>
    Sampled Position Offset Coordinate Space:ENiagaraSkelMeshTransforms=::NewEnumerator2
    Sampled Velocity Coordinate Space:ENiagaraCoordinateSpace=::Local
    Socket Index=1
    Socket Reader=$System.Socket Reader
    Socket Sampling Mode=::NewEnumerator3
    VelocitySpeedLimit=false
   )
   &ShapeLocation(MaskSpawnGroup=false, Override Seed=false, Sphere Radius=25, UseOwnerScale=true)
   &AddVelocity(
    Cone Angle=145
    Cone Axis=<0 0 1>
    Module.UseConeFalloff=false
    Module.UseVelDistribution=false
    Override Seed=false
    Velocity Mode=::NewEnumerator2
    Velocity Speed=&Multiply_Float(A=&RandomRangeFloat(Minimum=200, Maximum=300), B=1)
   ) {module=AddVelocity001}
  ]
  particleUpdateScript [
   &ParticleState()
   &ScaleSpriteSize(
    Scale Sprite Size Mode=::NewEnumerator0
    Uniform Scale Factor=&Multiply_Float(
     A=&FloatFromCurve(FloatCurve=DataInterface() {propertyValues {Curve {keys [
          {arriveTangent=10.9736, interpMode=RCIM_Cubic, leaveTangent=10.9736, tangentMode=RCTM_User, time=0, value=0}
          <at 0.1 0.95>
          {interpMode=RCIM_Cubic, tangentMode=RCTM_User, time=0.7, value=1}
          {tangentMode=RCTM_User, time=1, value=0}
         ]}}})
     B=&FloatFromCurve(
      FloatCurve=DataInterface() {propertyValues {Curve {
         defaultValue=1
         keys [
          {interpMode=RCIM_Cubic, tangentMode=RCTM_User, time=0, value=1}
          {arriveTangent=-2.4941, interpMode=RCIM_Cubic, leaveTangent=-2.4941, tangentMode=RCTM_User, time=1, value=0}
         ]
        }}}
      CurveIndex=$System.NormalizedLoopAge
     )
    )
   )
   &SpriteRotationRate(Rotation Rate=&Multiply_Float(
     A=&RandomRangeFloat(Minimum=10, Maximum=15)
     B=&MakeFloatFromVector(Vector=$Particles.Velocity, Channel=::NewEnumerator0)
    ))
   &GravityForce(Gravity=&RandomRangeVector(Minimum=<0 0 -150>, Maximum=<0 0 -200>))
   &Drag(
    Drag=&RandomRangeFloat(Minimum=2, Maximum=3)
    Rotational Drag:NiagaraFloat=&RandomRangeFloat(Minimum=5, Maximum=10)
    Use Linear Drag=true
    Use Rotational Drag=true
   )
   &CurlNoiseForce(Noise Frequency=15, Noise Strength=150, Pan Noise=true, Pan Noise Field=<0.5 0.5 -0.5>)
   &InheritSourceMovement(ApplyForce=false, ApplyPosition=true, ApplyPositionFalloff=false)
   &SolveForcesAndVelocity(Clamp Velocity=false, Limit Acceleration=false)
  ]
  rendererClasses [&NiagaraSpriteRendererProperties]
  renderers [{rendererClass=&NiagaraSpriteRendererProperties}]
 }
 MotionBlur {
  emitterSpawnScript [/SetVariables(Emitter.Fan Rotation:Vector3f=&FN_FrontEndVector3Modifier(Gameplay=<0 0 0>, Front End=<1 0 0>)) {isSetParameters=true}]
  emitterUpdateScript [
   &EmitterState(Life Cycle Mode=::NewEnumerator1, Loop Behavior=::NewEnumerator1, Loop Duration=1, UseLoopDelay=false)
   &SpawnBurst_Instantaneous(
    Spawn Time=&FN_FrontEndFloatModifier(Gameplay=0.5, Front End=0)
    Use Spawn Probability=false
    UseLoopCountLimit=false
   )
  ]
  particleSpawnScript [
   &InitializeParticle(
    Color=<rgba 0.5 0.5 0.5 1>
    Lifetime=&FN_FrontEndFloatModifier(Gameplay=0.8, Front End=1.45)
    Mesh Scale Mode=::NewEnumerator3
    Mesh Uniform Scale=0.4
    Sprite Size Mode=::NewEnumerator3
    Uniform Sprite Size=80
    UsePositionOffset=false
    Write Material Random=true
   )
   &InitialMeshOrientation(Mesh Orientation Mode=::NewEnumerator1, Rotation=<0.25 0 0>, Rotation Coordinate Space=::NewEnumerator2, Use Rotation Vector=true)
   /SetVariables(Particles.Scale=&VectorFromFloat(Value=&FN_FrontEndFloatModifier(Gameplay=0.38, Front End=0.49))) {isSetParameters=true}
  ]
  particleUpdateScript [
   &ParticleState()
   &SocketLocation(
    SamplePosOffset=true
    Sampled Position Offset Coordinate Space:ENiagaraSkelMeshTransforms=::NewEnumerator2
    Scale Sampling:ENiagara_AttributeSamplingApplyOutput=::NewEnumerator0
    Socket Index=2
    Socket Reader=$System.Socket Reader
    Socket Sampling Mode=::NewEnumerator3
    VelocitySpeedLimit=false
   )
   &DynamicMaterialParameters(
    Index 0 Param 1=0.15
    Index 0 Param 2=&FN_FrontEndFloatModifier(Gameplay=0.989, Front End=0.985)
    Index 0 Param 4=1
    Module.Index0Float0Write=true
    Module.Index0Float1Write=true
    Module.Index0Float2Write=true
    Module.Index0Float3Write=true
   )
   &ScaleColor(
    Scale Alpha=&Multiply_Float(
     A=&FloatFromCurve(
      FloatCurve=DataInterface() {propertyValues {Curve {
         defaultValue=1
         keys [<at 0 0 RCIM_Cubic>, {arriveTangent=1.7827, interpMode=RCIM_Cubic, leaveTangent=1.7827, tangentMode=RCTM_User, time=0.5, value=1}]
        }}}
      CurveIndex=$Emitter.Age
     )
     B=&FN_FrontEndFloatModifier(
      Gameplay=&FloatFromCurve(
       FloatCurve=DataInterface() {propertyValues {Curve {
          defaultValue=1
          keys [{interpMode=RCIM_Cubic, tangentMode=RCTM_User, time=0, value=1}, {interpMode=RCIM_Cubic, tangentMode=RCTM_User, time=1, value=0}]
         }}}
       Scale Curve=&Multiply_Float(A=2, B=$System.TOD)
      )
      Front End=&FloatFromCurve(
       FloatCurve=DataInterface() {propertyValues {Curve {
          defaultValue=1
          keys [
           {arriveTangent=0.1436, interpMode=RCIM_Cubic, leaveTangent=0.1436, tangentMode=RCTM_User, time=0, value=1}
           {arriveTangent=-1.7212, interpMode=RCIM_Cubic, leaveTangent=-2.8752, tangentMode=RCTM_Break, time=0.5891, value=0.4305}
           {arriveTangent=0.1067, interpMode=RCIM_Cubic, leaveTangent=0.1067, tangentMode=RCTM_User, time=1, value=0}
          ]
         }}}
       Scale Curve=&Multiply_Float(A=2, B=$System.TOD)
      )
     )
    )
    ScaleA=true
    ScaleRGB=true
   )
  ]
  rendererClasses [&NiagaraMeshRendererProperties]
  renderers [{rendererClass=&NiagaraMeshRendererProperties, rendererIndex=0}]
 }
}
summary {
 systemName=NS_Glider_CoolSlice_Deploy
 emitters [
  {emitterName=MotionBlur, rendererClasses [&NiagaraMeshRendererProperties]}
  {emitterName=Hairs, rendererClasses [&NiagaraSpriteRendererProperties]}
 ]
 userVariables [
  User.bIsFrontEnd {type=NiagaraBool}
  User.bIsFrontEndPreview {type=NiagaraBool}
  User.ForwardDot {type=NiagaraFloat, defaultValue=0, description=""}
 ]
}
