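# Niagara trail effect for the CoolSlice Golf glider, layering a socket-attached comb ribbon, falling hair sprites, speed lines and a wind mesh.
# The ribbon is a 50-particle burst re-placed each update, bent along its index by curves scaled with timelines and smoothed system values.
# Timelines and timers also drive spawn rates and fades.
# Front-end flags switch shape, motion and intensity from gameplay values, while time-of-day functions adjust color and brightness.
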NiagaraSystem /BRCosmetics/Gadgets/Assets/VinderTech_GliderChute/Glider_CoolSlice_Golf/FX/NS_Glider_CoolSlice_Trail
dependencies {
 usedDataInterfaces [
  &NiagaraDataInterfaceArrayInt32
  &NiagaraDataInterfaceCurve
  &NiagaraDataInterfaceArrayFloat
  &NiagaraDataInterfaceSpriteRendererInfo
  /Script/ChaosNiagara.NiagaraDataInterfacePhysicsField
  &NiagaraDataInterfaceSocketReader
 ]
 usedDynamicInputs [
  &RemapRange
  &MakeIntFromFloat
  &NormalizeVector
  &Add_Vector
  &TransformVector
 ]
 usedModules [
  &self:SystemSpawnScript.SystemScriptSource.SystemScriptGraph.NiagaraNodeAssignment_1.SetVariables_#1
  &SystemState
  &Timeline
  &FN_Integrator
  &FN_Timer
 ]
}
emitters {
 Comb_Trail {
  emitterUpdateScript [
   &EmitterState(Life Cycle Mode=::NewEnumerator1, Loop Behavior=::NewEnumerator1, Loop Delay=3, Loop Duration Mode=::NewEnumerator1, UseLoopDelay=true)
   &SpawnBurst_Instantaneous(Spawn Count=50, Use Spawn Probability=false, UseLoopCountLimit=true)
  ]
  particleSpawnScript [
   &PartitionParticles(
    Normalized Index In Partition Includes One=true
    Number Of Partitions=1
    Particles Per Partition=50
    Partition Mode:ENiagara_PartitionMode_v2=::NewEnumerator1
    Write Partition Index to Ribbon Link Order:ENiagara_AttributeSamplingApplyOutput=::NewEnumerator0
    Write Partition to Ribbon ID=::NewEnumerator0
   )
   /SetVariables(
    Particles.RibbonID=/Niagara/DynamicInputs/Meshes/MakeRibbonIDFromSpawnGroup()
    Particles.RibbonLinkOrder:NiagaraFloat=&ReturnNormalizedExecIndex()
    Particles.SpawnGroup=$Particles.PartitionParticles.Partition
   ) {isSetParameters=true}
   &InitializeParticle(
    Color Mode=::NewEnumerator0
    Lifetime=0.2
    Position Offset=<0 0 25>
    Ribbon Facing Mode=::NewEnumerator1
    Ribbon Facing Vector:Vector3f=<0 1 1>
    Ribbon Width=&MergeHelper_FloatSwitch(
     0=&MakeCustomFloatFromBool(Bool=$User.bIsFrontEndPreview, True Float=-50, False Float=70)
     1=&MakeCustomFloatFromBool(Bool=$User.bIsFrontEndPreview, True Float=50, False Float=56)
     2=&MakeCustomFloatFromBool(Bool=$User.bIsFrontEndPreview, True Float=25, False Float=50)
     3=&MakeCustomFloatFromBool(Bool=$User.bIsFrontEndPreview, True Float=25, False Float=50)
     Index=$Particles.PartitionParticles.Partition
     Num=::NewEnumerator0
    )
    Ribbon Width Mode=::NewEnumerator1
    Sprite Size Mode=::NewEnumerator3
    Uniform Sprite Size=5
    UsePositionOffset=false
    Write Material Random=true
   )
  ]
  particleUpdateScript [
   &ParticleState(Kill Particles When Lifetime Has Elapsed=false, Let Infinitely Lived Particles Die When Emitter Deactivates=true)
   &SocketLocation(
    Interpolate Socket Sampling=true
    SamplePosOffset=true
    Sampled Position Coordinate Space:ENiagaraCoordinateSpace=::Local
    Socket Index=&SelectIntFromArray(Direct Array Index=$Particles.PartitionParticles.Partition, Int Selection Array=DataInterface() {propertyValues {IntData [2,2]}})
    Socket Reader=$System.Socket Reader
    Socket Sampling Mode=::NewEnumerator3
    VelocitySpeedLimit=false
   )
   /SetVariables(Particles.Position=&LerpPosition(
     A=$Particles.Position
     B=&AddVectorToPosition(
      Position=$Particles.Position
      Vector=&MakeCustomVectorFromBool(
       Bool=$User.bIsFrontEndPreview
       True Vector=<0 200 -100>
       False Vector=&MakeVector(
        X=&FloatFromCurve(
         FloatCurve=DataInterface() {propertyValues {Curve {keys [<at 0 0 RCIM_Cubic>, {arriveTangent=2.1213, interpMode=RCIM_Cubic, leaveTangent=2.1213, tangentMode=RCTM_User, time=1, value=1}]}}}
         CurveIndex=&ReturnNormalizedExecIndex()
         Scale Curve=&Multiply_Float(
          A=&MakeFloatFromVector(Vector=$System.InterpolateOverTime.InterpolateOverTime_V3_.Moving Average, Channel=::NewEnumerator0)
          B=&Lerp_Float(A=0, B=50, Alpha=$System.InterpolateOverTime001.InterpolateOverTime_Float.Moving Average)
         )
        )
        Y=&FloatFromCurve(
         FloatCurve=DataInterface() {propertyValues {Curve {keys [{tangentMode=RCTM_User, time=0, value=0}, {tangentMode=RCTM_User, time=1, value=1}]}}}
         CurveIndex=&ReturnNormalizedExecIndex()
         Scale Curve=&FN_FrontEndFloatModifier(
          Gameplay=&Lerp_Float(
           A=&SelectFloatFromArray(Direct Array Index=$Particles.PartitionParticles.Partition, Float Selection Array=DataInterface() {propertyValues {FloatData [50,50,50,50]}})
           B=&SelectFloatFromArray(
            Direct Array Index=$Particles.PartitionParticles.Partition
            Float Selection Array=DataInterface() {propertyValues {FloatData [150,150,150,150]}}
           )
           Alpha=$System.Timeline.NormalizedTime
          )
          Front End=&Lerp_Float(A=0, B=200, Alpha=$System.Timeline002.NormalizedTime)
         )
        )
        Z=0
       )
      )
     )
     Alpha=&ReturnNormalizedExecIndex()
    )) {isSetParameters=true}
   &ScaleRibbonWidth(Ribbon Width Scale=&MakeCustomFloatFromBool(
     Bool=&SetBoolByFloatComparison(
      A=&MakeCustomFloatFromBool(Bool=$User.bIsFrontEndPreview, True Float=0, False Float=$System.Timeline001.NormalizedTime)
      B=&Multiply_Float(A=$System.Timeline.Time, B=4)
     )
     True Float=0
     False Float=&FloatFromCurve(
      FloatCurve=DataInterface() {propertyValues {Curve {defaultValue=0.5, keys [{tangentMode=RCTM_User, time=0, value=0.5}, {tangentMode=RCTM_User, time=1, value=1}]}}}
      CurveIndex=$Particles.Initial.RibbonLinkOrder
     )
    ))
   &DynamicMaterialParameters(
    Index 1 Param 1=&Multiply_Float(
     A=&FN_FrontEndFloatModifier(
      Gameplay=&FloatFromCurve(
       FloatCurve=DataInterface() {propertyValues {Curve {keys [{tangentMode=RCTM_User, time=0, value=1}, {tangentMode=RCTM_User, time=1, value=-0.1}]}}}
       CurveIndex=$Particles.Initial.RibbonLinkOrder
       Scale Curve=&FN_FrontEndFloatModifier(
        Gameplay=&Lerp_Float(A=0, B=1, Alpha=$System.Timeline.NormalizedTime)
        Front End=&Lerp_Float(A=0, B=1, Alpha=$System.Timeline002.NormalizedTime)
       )
      )
      Front End=&FloatFromCurve(
       FloatCurve=DataInterface() {propertyValues {Curve {keys [{tangentMode=RCTM_User, time=0, value=1}, {tangentMode=RCTM_User, time=1, value=-0.1}]}}}
       CurveIndex=$Particles.Initial.RibbonLinkOrder
       Scale Curve=&FN_FrontEndFloatModifier(
        Gameplay=&Lerp_Float(A=0, B=1, Alpha=$System.Timeline.NormalizedTime)
        Front End=&Lerp_Float(A=0, B=1, Alpha=$System.Timeline002.NormalizedTime)
       )
      )
     )
     B=&FNTimeOfDayScalar(
      Night Scale=0.6
      Morning Scale=0.5
      Day Scale=&FN_FrontEndFloatModifier(Gameplay=1, Front End=0.3)
      Evening Scale=0.6
     )
    )
    Module.Index1Float0Write=true
    Module.Index1Float1Write=true
    Module.Index1Float2Write=true
    Module.Index1Float3Write=true
    Write Parameter Index 0=::NewEnumerator0
    Write Parameter Index 1=::NewEnumerator1
   )
  ]
  rendererClasses [&NiagaraRibbonRendererProperties]
  renderers [{rendererClass=&NiagaraRibbonRendererProperties}]
 }
 Hairs {
  emitterUpdateScript [
   &EmitterState(
    Delay First Loop Only=true
    Inactive Response=::NewEnumerator1
    Life Cycle Mode=::NewEnumerator1
    Loop Delay=3
    Loop Duration=3
    UseLoopDelay=true
   )
   &SpawnRate(
    Spawn Probability=0.5
    SpawnRate=&FN_FrontEndFloatModifier(
     Gameplay=&Lerp_Float(A=3, B=6, Alpha=$System.Timeline.NormalizedTime)
     Front End=&Lerp_Float(A=3, B=4, Alpha=$System.Timeline002.NormalizedTime)
    )
    Use Spawn Probability=true
   )
  ]
  particleSpawnScript [
   &InitializeParticle(
    Lifetime Max=3
    Lifetime Min=2
    Lifetime Mode=::NewEnumerator1
    Sprite Rotation Mode=::NewEnumerator2
    Sprite Size=&Multiply_Vector2D_ByFloat(Float=&RandomRangeFloat(Minimum=15, Maximum=18), Vector2D=<1 2>)
    Sprite Size Mode=::NewEnumerator4
    UsePositionOffset=false
    Write Material Random=true
   )
   &SubUVAnimation(
    End Frame Range Override=2
    "Sprite Renderer "=DataInterface() {propertyValues {SpriteRenderer=&self:Hairs_1.NiagaraSpriteRendererProperties_0}}
    Start Frame Range Override=1
    SubUV Animation Mode:ENiagara_SubUVLookupModeV2=::NewEnumerator2
    UseEndFrame=true
    UseStartFrame=true
   )
   &SocketLocation(
    SamplePosOffset=true
    Sampled Position Offset Coordinate Space:ENiagaraSkelMeshTransforms=::NewEnumerator2
    Socket Reader=$System.Socket Reader
    Socket Sampling Mode=::NewEnumerator3
    VelocitySpeedLimit=false
   )
   &ShapeLocation(
    Cylinder Height=25
    MaskSpawnGroup=false
    Offset=<0 0 25>
    Offset Mode=::NewEnumerator0
    Override Seed=false
    Shape Primitive=::NewEnumerator1
    Surface Only=false
    UseOwnerScale=true
   )
   &AddVelocity(
    Velocity=&FN_FrontEndVector3Modifier(Gameplay=&RandomRangeVector(Minimum=<-250 200 0>, Maximum=<250 300 0>), Front End=<0 400 0>)
    Velocity Speed Scale=&FN_FrontEndFloatModifier(Gameplay=1, Front End=1)
   )
   &AccelerationForce()
   &ApplyInitialForces()
  ]
  particleUpdateScript [
   &ParticleState()
   &ScaleSpriteSize(
    Scale Sprite Size Mode=::NewEnumerator0
    Uniform Scale Factor=&Multiply_Float(
     A=&FloatFromCurve(FloatCurve=DataInterface() {propertyValues {Curve {keys [
          {arriveTangent=10.9736, interpMode=RCIM_Cubic, leaveTangent=10.9736, tangentMode=RCTM_User, time=0, value=0}
          <at 0.1 0.95>
          {interpMode=RCIM_Cubic, tangentMode=RCTM_User, time=0.7, value=1}
          {tangentMode=RCTM_User, time=1, value=0}
         ]}}})
     B=1
    )
   )
   &SpriteRotationRate(Rotation Rate=&Multiply_Float(A=&RandomRangeFloat(Minimum=-50, Maximum=-40), B=5))
   &InheritSourceMovement(
    Applied Position Falloff Curve=DataInterface() {propertyValues {Curve {
       defaultValue=1
       keys [
        {arriveTangent=2.4263, interpMode=RCIM_Cubic, leaveTangent=2.4263, tangentMode=RCTM_User, time=0, value=0}
        {tangentMode=RCTM_User, time=1, value=1}
       ]
      }}}
    Applied Position Falloff Distance=25
    Applied Position Scale=&VectorFromFloat(Value=&FloatFromCurve(FloatCurve=DataInterface() {propertyValues {Curve {
         defaultValue=1
         keys [{interpMode=RCIM_Cubic, tangentMode=RCTM_User, time=0, value=1}, {interpMode=RCIM_Cubic, tangentMode=RCTM_User, time=1, value=0.75}]
        }}}))
    ApplyForce=false
    ApplyPosition=true
    ApplyPositionFalloff=true
    ApplyPositionFalloffCurve=true
   )
   &GravityForce(Gravity=&FN_FrontEndVector3Modifier(Gameplay=<0 0 -2000>, Front End=<0 0 -1600>))
   &WindForce(Use Physics Field=true, Wind Speed=&RandomRangeVector(Minimum=<-30 -30 100>, Maximum=<30 30 150>))
   &Drag(
    Drag=&RandomRangeFloat(Minimum=5, Maximum=10)
    Rotational Drag:NiagaraFloat=&RandomRangeFloat(Minimum=0.5, Maximum=1)
    Use Linear Drag=true
    Use Rotational Drag=true
   )
   &SolveForcesAndVelocity(Clamp Velocity=false, Limit Acceleration=false)
   &ScaleColor(
    Scale RGB=&MakeVectorFromLinearColorRGB(Color=&FNTimeOfDayColor(
      Night Color=<rgba 0.3 0.3608 1 1>
      Morning Color=<rgba 1.25 1.25 0.375 1>
      Day Color=&MakeCustomLinearColorFromBool(
       Bool=&Boolean_OrOperation(Bool1=$User.bIsFrontEnd, Bool2=$User.bIsFrontEndPreview)
       True Color=<rgba 0.6145 0.6145 0.6145 1>
       False Color=<rgba 1 1 1 1>
      )
      Evening Color=<rgba 1 0.8017 0.4051 1>
     ))
    ScaleA=true
    ScaleRGB=true
   )
  ]
  rendererClasses [&NiagaraSpriteRendererProperties]
  renderers [{rendererClass=&NiagaraSpriteRendererProperties}]
 }
 SpeedLine {
  emitterUpdateScript [
   &EmitterState(
    Delay First Loop Only=true
    Inactive Response=::NewEnumerator1
    Life Cycle Mode=::NewEnumerator1
    Loop Delay=3
    Loop Duration=3
    UseLoopDelay=true
   )
   &SpawnRate(
    SpawnRate=&FN_FrontEndFloatModifier(
     Gameplay=&Multiply_Float(
      A=&Multiply_Float(A=200, B=1)
      B=&Lerp_Float(A=-0.8, B=1, Alpha=$System.FN_Timer.Value)
     )
     Front End=90
    )
    Use Spawn Probability=false
   )
  ]
  particleSpawnScript [
   &InitializeParticle(
    Lifetime Max=0.12
    Lifetime Min=0.1
    Lifetime Mode=::NewEnumerator1
    Position Offset=&FN_FrontEndVector3Modifier(Gameplay=<0 0 0>, Front End=<-30 35 50>)
    Sprite Size Max=<2 7>
    Sprite Size Min=<1.75 2>
    Sprite Size Mode=::NewEnumerator2
    UsePositionOffset=true
    Write Material Random=true
   )
   &TorusLocation(
    Align Torus With Axis=true
    Handle Radius=0
    Large Radius:NiagaraFloat=&RandomRangeFloat(
     Minimum=&FN_FrontEndFloatModifier(Gameplay=150, Front End=150)
     Maximum=&FN_FrontEndFloatModifier(Gameplay=200, Front End=220)
    )
    MaskSpawnGroup=false
    Module.Override Randomness=false
    Offset=&FN_FrontEndVector3Modifier(Gameplay=<0 100 0>, Front End=<0 -35 0>)
    Torus Axis=&FN_FrontEndVector3Modifier(Gameplay=<0 1 0>, Front End=<0.5 1 0>)
    Torus Origin:NiagaraPosition=&AddVectorToPosition(
     Position=$Particles.Position
     Vector=&FN_FrontEndVector3Modifier(Gameplay=<0 0 0>, Front End=<0 0 -50>)
    )
   )
   &AddVelocityFromPoint(
    Module.Use Falloff Distance=false
    Origin Offset=&FN_FrontEndVector3Modifier(Gameplay=<0 -200 0>, Front End=<0 -400 0>)
    Velocity Strength=&FN_FrontEndFloatModifier(Gameplay=1500, Front End=300)
   )
   &AddVelocity(
    Velocity=&FN_FrontEndVector3Modifier(Gameplay=<0 500 0>, Front End=<0 500 -300>)
    Velocity Speed Scale=&FN_FrontEndFloatModifier(Gameplay=1, Front End=1)
   )
  ]
  particleUpdateScript [
   &SolveForcesAndVelocity(Clamp Velocity=false, Limit Acceleration=false)
   &ParticleState()
   &ScaleSpriteSizeBySpeed(Max Scale Factor=<1 50>, Min Scale Factor=<0.01 0.05>, Velocity Threshold=1)
   &ScaleColor(
    Scale Alpha=&Multiply_Float(
     A=&FloatFromCurve(
      FloatCurve=DataInterface() {propertyValues {Curve {keys [
          {interpMode=RCIM_Cubic, tangentMode=RCTM_User, time=0, value=0}
          {interpMode=RCIM_Cubic, tangentMode=RCTM_User, time=0.5, value=1}
          {tangentMode=RCTM_User, time=1, value=0}
         ]}}}
      Scale Curve=&FNTimeOfDayScalar(Night Scale=0.6, Morning Scale=0.8, Evening Scale=0.8)
     )
     B=&Multiply_Float(
      A=&RandomRangeFloat(Minimum=0.1, Maximum=0.3)
      B=&FN_FrontEndFloatModifier(Gameplay=1, Front End=0.3)
     )
    )
    Scale RGB=&MakeVectorFromLinearColorRGB(Color=&FNTimeOfDayColor(Night Color=<rgba 0.3478 0.5093 1 1>, Morning Color=<rgba 1 0.621 0.469 1>, Day Color=<rgba 1 1 1 1>, Evening Color=<rgba 1 0.621 0.469 1>))
    ScaleA=true
    ScaleRGB=true
   )
  ]
  rendererClasses [&NiagaraSpriteRendererProperties]
  renderers [{rendererClass=&NiagaraSpriteRendererProperties}]
 }
 Wind {
  emitterUpdateScript [
   &EmitterState(
    Inactive Response=::NewEnumerator1
    Life Cycle Mode=::NewEnumerator1
    Loop Behavior=::NewEnumerator1
    Loop Delay=3
    Loop Duration Mode=::NewEnumerator1
    UseLoopDelay=true
   )
   &SpawnBurst_Instantaneous(Use Spawn Probability=false, UseLoopCountLimit=true)
  ]
  particleSpawnScript [
   &InitializeParticle(
    Lifetime=0.5
    Mesh Scale=&Multiply_VectorByFloat(Vector=&FN_FrontEndVector3Modifier(Gameplay=<0.7 0.5 0.3>, Front End=<0.55 0.4 0.2>), Float=2.5)
    Mesh Scale Mode=::NewEnumerator4
    Sprite Size Mode=::NewEnumerator3
    Uniform Sprite Size=999
    UsePositionOffset=true
    Write Material Random=true
   )
   &InitialMeshOrientation(
    Mesh Orientation Mode=::NewEnumerator7
    Rotation=&FN_FrontEndVector3Modifier(Gameplay=<0.52 0 0>, Front End=<0.5 0 0>)
    Use Rotation Vector=true
   )
  ]
  particleUpdateScript [
   &ParticleState(Kill Particles When Lifetime Has Elapsed=false)
   &SocketLocation(
    SamplePosOffset=true
    Sampled Position Offset=&FN_FrontEndVector3Modifier(Gameplay=<0 -6 25>, Front End=<0 -7 20>)
    Socket Reader=$System.Socket Reader
    Socket Sampling Mode=::NewEnumerator3
    VelocitySpeedLimit=false
   )
   &InterpolateOverTime(Rate Of Change=15, Target Quat Value:Quat4f=$Output.SocketLocation.SampledRotation, Variable Type=::NewEnumerator5)
   /SetVariables(Particles.MeshOrientation=&MultiplyQuaternion(Quaternion A=$StackContext.InterpolateOverTime.InterpolateOverTime_Quaternion.Moving Average, Quaternion B=$StackContext.InitialRotation)) {isSetParameters=true}
   &ScaleColor(
    Scale Alpha=&Multiply_Float(
     A=&FN_FrontEndFloatModifier(Gameplay=0.2, Front End=0.13)
     B=&FN_FrontEndFloatModifier(
      Gameplay=&Lerp_Float(A=0, B=1, Alpha=$System.Timeline.NormalizedTime)
      Front End=&Lerp_Float(A=0, B=1, Alpha=$System.Timeline002.NormalizedTime)
     )
    )
    Scale RGB=&MakeVectorFromLinearColorRGB(Color=&FNTimeOfDayColor(Night Color=<rgba 0.3478 0.5093 1 1>, Morning Color=<rgba 1 0.621 0.469 1>, Day Color=<rgba 1 1 1 1>, Evening Color=<rgba 1 0.621 0.469 1>))
    ScaleA=true
    ScaleRGB=true
   ) {module=ScaleColor001}
  ]
  rendererClasses [&NiagaraMeshRendererProperties]
  renderers [{rendererClass=&NiagaraMeshRendererProperties, rendererIndex=0}]
 }
}
summary {
 systemName=NS_Glider_CoolSlice_Trail
 emitters [
  {emitterName=Wind, rendererClasses [&NiagaraMeshRendererProperties]}
  {emitterName=Comb_Trail, rendererClasses [&NiagaraRibbonRendererProperties]}
  {emitterName=Hairs, rendererClasses [&NiagaraSpriteRendererProperties]}
  {emitterName=SpeedLine, rendererClasses [&NiagaraSpriteRendererProperties]}
 ]
 userVariables [
  User.bIsFrontEnd {type=NiagaraBool}
  User.bIsFrontEndPreview {type=NiagaraBool}
  User.bIsFullyDeployed {type=NiagaraBool}
  User.ForwardDot {type=NiagaraFloat, defaultValue=0, description=""}
  User.RibbonIDOffset {type=NiagaraInt32}
  User.RightDot {type=NiagaraFloat, defaultValue=0, description=""}
 ]
}
